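# Conversation database for the IronLilac NPC, providing its greeting and the hub into its other conversations.
# The entry point leads to one speech node with a default line and contextual lines for first-ever meetings, match conversation count and specific player outfits.
# Contextual lines carry priorities, mostly from named Tandem speech curve rows.
# The speech node outputs to link nodes for basic services, IronLilac dialogue and two Season38 Hope entries.
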ConversationDatabase /NPCLibrary/NPCs/IronLilac/NPCLibrary_Conversation_IronLilac
entry_points [{tag=Conversation.Entry.IronLilac, destinations [#1]}]
nodes [
 ConversationEntryPointNode_7 {
  id=#1
  class=ConversationEntryPointNode
  outputs [#2]
  properties {entryTag {tagName=Conversation.Entry.IronLilac}}
 }
 Speech_C_0 {
  id=#2
  class=Speech_C
  outputs [#3, #4, #5, #6]
  properties {
   nodeName=""
   generalConfig {
    defaultMessage="Strange days, huh?"
    contextualMessages [
     {
      message="Hey, Looperino... I think that's what we're saying these days?"
      contextRequirements [{requirement {class=FortControllerRequirement_IsFirstEverConversationWithNPC}}]
      priority {curve {rowName=Tandem.Speech.Priority.FirstEverInteraction}}
     }
     {
      message="Strange days, huh?"
      contextRequirements [{requirement {class=FortControllerRequirement_CheckNPCMatchConversationCount, properties {nPCUniqueIdTag {tagName=AISpawnerData.Type.Tandem.IronLilac}}}}]
      priority {curve {rowName=None}}
     }
     {
      message="The island needs our help. We have to find our way back before it's too late. "
      contextRequirements [{requirement {
         class=FortControllerRequirement_HasCID
         properties {allowedCIds [&Character_UndergroundRebel_Fashion]}
        }}]
      priority {value=2, curve {rowName=Tandem.Speech.Priority.OutfitReaction}}
     }
     {
      message=Let's hope you and Jones never join forces...
      contextRequirements [{requirement {class=FortControllerRequirement_HasCID, properties {allowedCIds [/BRCosmetics/Athena/Items/Cosmetics/Characters/Character_QueenTruth_Rind]}}}]
      priority {value=2, curve {rowName=Tandem.Speech.Priority.OutfitReaction}}
     }
     {
      message=I heard you play a mean sax.
      contextRequirements [{requirement {class=FortControllerRequirement_HasCID, properties {allowedCIds [/BRCosmetics/Athena/Items/Cosmetics/Characters/Character_FrostGalore]}}}]
      priority {value=2, curve {rowName=Tandem.Speech.Priority.OutfitReaction}}
     }
    ]
   }
  }
 }
 ConversationLinkNode_2 {id=#6, class=ConversationLinkNode, properties {remoteEntryTag {tagName=Conversation.Entry.Season38.Hope.P04}}}
 ConversationLinkNode_1 {id=#5, class=ConversationLinkNode, properties {remoteEntryTag {tagName=Conversation.Entry.Season38.Hope}}}
 ConversationLinkNode_6 {id=#4, class=ConversationLinkNode, properties {remoteEntryTag {tagName=Conversation.Entry.IronLilac.Dialogue}}}
 ConversationLinkNode_0 {id=#3, class=ConversationLinkNode, properties {remoteEntryTag {tagName=Conversation.Entry.BasicServices}}}
]
