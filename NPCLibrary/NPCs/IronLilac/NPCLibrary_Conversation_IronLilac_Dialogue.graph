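# Conversation database for the Iron Lilac NPC's dialogue, reached through the Conversation.Entry.IronLilac.Dialogue entry tag.
# An "Ask about..." choice leads to a speech offering help, which branches to a topic on the island or a Back node.
# The island topic gives a worried line about the danger of not finding a way back, then offers a choice that returns to the start.
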ConversationDatabase /NPCLibrary/NPCs/IronLilac/NPCLibrary_Conversation_IronLilac_Dialogue
entry_points [{tag=Conversation.Entry.IronLilac.Dialogue, destinations [#1]}]
nodes [
 ConversationEntryPointNode_0 {
  id=#1
  class=ConversationEntryPointNode
  outputs [#2]
  properties {entryTag {tagName=Conversation.Entry.IronLilac.Dialogue}}
 }
 Speech_C_0 {
  id=#2
  class=Speech_C
  outputs [#3, #4]
  properties {
   nodeName=""
   softTaskIcon=&T_UI_NPC_question
   generalConfig {defaultMessage="I don't have all the answers, but I'll try to help."}
  }
  sub_nodes [ConversationChoiceNode_5 {class=ConversationChoiceNode, properties {defaultChoiceDisplayText=Ask about...}}]
 }
 Speech_C_11 {
  id=#4
  class=Speech_C
  outputs [#5]
  properties {
   nodeName=""
   softTaskIcon=&T_UI_NPC_question
   generalConfig {defaultMessage="I'm worried if we don't find our way back soon, things will get very, very dangerous."}
  }
  sub_nodes [ConversationChoiceNode_3 {class=ConversationChoiceNode, properties {defaultChoiceDisplayText=The island...}}]
 }
 ReturnToStart_C_23 {
  id=#5
  class=ReturnToStart_C
  properties {softTaskIcon=&T_UI_NPC_Back}
  sub_nodes [ConversationChoiceNode_32 {class=ConversationChoiceNode, properties {defaultChoiceDisplayText=BACK}}]
 }
 Back_C_1 {
  id=#3
  class=Back_C
  sub_nodes [ConversationChoiceNode_2 {class=ConversationChoiceNode, properties {defaultChoiceDisplayText=BACK}}]
 }
]
